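# Translucent, unlit, two-sided ribbon material for CoyoteTrail character body FX, enabled for Niagara mesh particles.
# SplineThicken sets width from a Spline Width parameter, added to time-animated sine waves along local X and Z for world position offset.
# A trail texture, distorted by panning Perlin noise and faded by sine masks on both UV axes, multiplies particle alpha for opacity.
# Particle color drives emissive through MF_FX_EmissiveAdjustments.
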Material /BRCosmetics/Characters/Player/Male/Medium/Bodies/M_MED_CoyoteTrail/FX/M_CoyoteTrail_Ribbon
graph {
 properties {
  bUsedWithNiagaraMeshParticles=true
  blendMode=BLEND_Translucent
  pixelDepthOffsetMode=PDOM_Legacy
  shadingModel=MSM_Unlit
  twoSided=true
  usedShadingModels=Unlit
 }
 nodes [
  n1 = ScalarParameter() {defaultValue=15, parameterName=Spline Width}
  n2 = Constant() {r=0.3}
  n3 = Add(A=ComponentMask(None=TextureCoordinate()) {r=true}, B=Multiply(A=Time(), B=@n2))
  n4 = Transform(None=Multiply(A=Constant3Vector() {constant=<rgba 0 0 1 0>}, B=Sine(None=@n3) {period=0.3})) {transformSourceType=TRANSFORMSOURCE_Local}
  n5 = Constant() {r=4}
  n6 = Add(A=ComponentMask(None=TextureCoordinate()) {r=true}, B=Multiply(A=@n2, B=Time()))
  n7 = Transform(None=Multiply(A=Constant3Vector() {constant=<rgba 1 0 0 0>}, B=Sine(None=@n6) {period=0.3})) {transformSourceType=TRANSFORMSOURCE_Local}
  Add(
   A=MaterialFunctionCall(WidthBase=@n1, WidthTip=@n1, Expand U or V UV Channel=StaticBool()) {materialFunction=&SplineThicken}.WorldPositionOffset
   B=Add(A=Multiply(A=@n4, B=@n5) {constB=10}, B=Multiply(A=@n7, B=@n5) {constB=10}) {constB=0}
  )
  n8 = ParticleColor()
  MaterialFunctionCall(Emissive=@n8.RGB) {materialFunction=&MF_FX_EmissiveAdjustments}
  n9 = Add(A=TextureCoordinate() {vTiling=0.5}, B=Constant2Vector() {g=0.25}) {constB=0}
  n10 = Panner(Coordinate=TextureCoordinate() {vTiling=0.2}) {speedY=0.1}
  n11 = Multiply(A=ComponentMask(None=TextureSample(UVs=@n10) {samplerType=SAMPLERTYPE_LinearGrayscale, texture=&T_SimplePerlin_Greyscale}.R) {g=true, r=true}) {constB=0.2}
  n12 = Sine(None=ComponentMask(None=TextureCoordinate()) {g=true}) {period=2}
  n13 = Multiply(
   A=TextureSample(UVs=Add(A=@n9, B=@n11)) {texture=/BRCosmetics/Weapons/FORT_Melee/Pickaxe_Tower_Sentinel_Female/FX/T_TowerSentinel_trail}.R
   B=Saturate(None=Multiply(A=@n12) {constB=3})
  )
  n14 = Sine(None=ComponentMask(None=TextureCoordinate()) {r=true}) {period=2}
  Multiply(A=@n8.A, B=Multiply(A=Saturate(None=@n13), B=Power(Base=@n14)))
 ]
}
