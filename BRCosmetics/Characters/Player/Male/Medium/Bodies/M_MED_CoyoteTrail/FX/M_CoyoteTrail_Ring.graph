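# Translucent, unlit, two-sided ring effect material for Niagara mesh particles in the CoyoteTrail character cosmetic.
# A Color-tinted trail texture, multiplied by particle color, feeds the emissive-adjustment function.
# A second trail mask, shaped by sine gradients along both UV axes and particle alpha, forms opacity, depth-faded on default and deferred paths but not forward or mobile.
# Panning noise, caustics and many other texture samples sit unused.
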Material /BRCosmetics/Characters/Player/Male/Medium/Bodies/M_MED_CoyoteTrail/FX/M_CoyoteTrail_Ring
graph {
 properties {
  bUsedWithNiagaraMeshParticles=true
  blendMode=BLEND_Translucent
  pixelDepthOffsetMode=PDOM_Legacy
  shadingModel=MSM_Unlit
  twoSided=true
  usedShadingModels=Unlit
 }
 nodes [
  n1 = ParticleColor()
  n2 = Add(A=Add(A=TextureCoordinate() {vTiling=0.6}, B=Constant2Vector() {g=0.2})) {constB=0}
  n3 = Multiply(
   A=VectorParameter() {parameterName=Color, defaultValue=<rgba 0 0.5 0.2755 1>}.RGB
   B=TextureSample(UVs=@n2) {texture=/BRCosmetics/Weapons/FORT_Melee/Pickaxe_Future_Samurai_Summer_Male/FX/T_Pickaxe_Future_Samurai_Summer_Male_Trail_03}.R
  )
  n4 = Constant()
  MaterialFunctionCall(Emissive=Multiply(A=@n1.RGB, B=Add(A=@n3, B=@n4))) {materialFunction=&MF_FX_EmissiveAdjustments}
  n5 = Panner(Coordinate=TextureCoordinate()) {speedY=0.3}
  Multiply(A=ComponentMask(None=TextureSample(UVs=@n5) {samplerType=SAMPLERTYPE_Grayscale, texture=&LowResBlurredNoise}.RGB) {g=true, r=true}) {constB=0.3}
  n6 = Panner(Coordinate=TextureCoordinate()) {speedY=0.3}
  Multiply(A=ComponentMask(None=TextureSample(UVs=@n6) {samplerType=SAMPLERTYPE_Grayscale, texture=&LowResBlurredNoise}.RGB) {g=true, r=true}) {constB=0.2}
  Power(Base=TextureSample(UVs=@n2) {texture=&T_ShieldCaustics}.R) {constExponent=1}
  n7 = TextureCoordinate()
  n8 = Sine(None=ComponentMask(None=TextureCoordinate()) {g=true}) {period=2}
  n9 = Multiply(
   A=TextureSample(UVs=Add(A=@n7) {constB=0}) {texture=/BRCosmetics/Weapons/FORT_Melee/Pickaxe_Future_Samurai_Summer_Male/FX/T_Pickaxe_Future_Samurai_Summer_Male_Trail_01}.R
   B=Saturate(None=Multiply(A=@n8) {constB=3})
  )
  n10 = Sine(None=ComponentMask(None=TextureCoordinate()) {r=true}) {period=2}
  n11 = Multiply(A=@n1.A, B=Multiply(A=Saturate(None=@n9), B=Power(Base=@n10)))
  n12 = DepthFade(Opacity=@n11) {fadeDistanceDefault=20}
  ShadingPathSwitch(Default=@n12, Deferred=@n12, Forward=@n11, Mobile=@n11)
  TextureSample() {texture=/BRCosmetics/Athena/Effects/Particles/Skydiving/NS_Contrail_Chainmail/T_Chainmail_WispyMaskFat}
  TextureSample() {samplerType=SAMPLERTYPE_LinearColor, texture=/BRCosmetics/Athena/Effects/Particles/Skydiving/NS_Contrail_CubeNinja/T_CubeNinja_Trail}
  TextureSample() {texture=/BRCosmetics/Weapons/FORT_Melee/Pickaxe_Gimmick_Female/FX/T_Pickaxe_Gimmick_Female_Trail_03}
  TextureSample() {texture=/BRCosmetics/Weapons/FORT_Melee/Pickaxe_Sunlit/FX/T_Pickaxe_Sunlit_Trail_01}
  TextureSample() {samplerType=SAMPLERTYPE_Grayscale, texture=&T_Smoke_Trail_Soft}
  TextureSample() {texture=/BRCosmetics/Weapons/FORT_Melee/Pickaxe_Tower_Sentinel_Female/FX/T_TowerSentinel_trail}
  TextureSample() {texture=/BRCosmetics/Athena/Effects/Particles/Skydiving/NS_Contrail_Watercolor/T_Watercolor_Trail}
  TextureSample() {
   samplerType=SAMPLERTYPE_LinearColor
   texture=/BRCosmetics/Athena/Effects/Particles/Skydiving/NS_Contrail_Turtleneck_Crystal/T_WavePattern_Chromatic
  }
  TextureSample(UVs=MaterialFunctionCall(XY=TextureCoordinate()) {materialFunction=&Swizzle}.YX) {
   samplerType=SAMPLERTYPE_LinearGrayscale
   texture=/BRCosmetics/Athena/Effects/Particles/Skydiving/NS_Contrail_DarkAndLight/T_DarkAndLight_Noise
  }
  TextureSample() {samplerType=SAMPLERTYPE_LinearColor, texture=/BRCosmetics/Weapons/FORT_Melee/Pickaxe_Mouse/FX/Textures/T_MouseTrail_M}
  TextureSample() {texture=/BRCosmetics/Weapons/FORT_Melee/Pickaxe_Future_Samurai_Summer_Male/FX/T_Pickaxe_Future_Samurai_Summer_Male_Trail_02}
  TextureSample(UVs=TextureCoordinate()) {texture=/BRCosmetics/Weapons/FORT_Melee/Pickaxe_Future_Samurai_Summer_Male/FX/T_Pickaxe_Future_Samurai_Summer_Male_Trail_04}
  TextureSample() {texture=/BRCosmetics/Weapons/FORT_Melee/Pickaxe_Gimmick_Female/FX/T_Pickaxe_Gimmick_Female_Trail_01}
  TextureSample() {samplerType=SAMPLERTYPE_LinearColor, texture=/BRCosmetics/Weapons/FORT_Melee/Pickaxe_Gimmick_Female/FX/T_Pickaxe_Gimmick_Female_Trail_02}
 ]
}
