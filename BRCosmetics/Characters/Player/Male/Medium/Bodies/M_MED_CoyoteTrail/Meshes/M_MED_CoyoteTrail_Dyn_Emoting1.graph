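# Physics asset for the CoyoteTrail male medium body, named as an emoting variant, defining secondary motion for skirt and rope bones.
# Seven skirt chains hang from spine_01 and two rope chains from pelvis, with joint limits generally widening down the skirt.
# Cross-constraints tie neighboring skirt chains together, some fully free in rotation, so the strands behave like one sheet.
# Capsules cover lower body, legs, feet and chain bones, plus a huge low collision_01 capsule.
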PhysicsAsset /BRCosmetics/Characters/Player/Male/Medium/Bodies/M_MED_CoyoteTrail/Meshes/M_MED_CoyoteTrail_Dyn_Emoting1
bodies {
 calf_l [{
   shapeType=Capsule
   length=44.8129
   radius=12.283
   shapeName=calf_l_capsule
   center=<-19.6647 -0.0731 2.1891>
   rotation=<rot -90 180 178.3097>
  }]
 calf_r [{shapeType=Capsule, length=44.8752, radius=14.5603, shapeName=calf_l_capsule, center=<25.1541 0.3327 -2.6783>, rotation=<rot -90 180 180>}]
 collision_01 [{shapeType=Capsule, length=118, radius=243.25, shapeName=collision_01_capsule, center=<0 0 -236.0728>, rotation=<rot -90 0 0>}]
 dyn_Rope_l_01 [{
   shapeType=Capsule
   length=8.755
   radius=1
   shapeName=dyn_Rope_l_01_capsule
   center=<0.2049 0.7771 -4.514>
   rotation=<rot -6.8919 -90 -4.9969>
  }]
 dyn_Rope_l_02 [{
   shapeType=Capsule
   length=8.505
   radius=1.25
   shapeName=dyn_Rope_l_02_capsule
   center=<0.1347 -0.0808 -1.8145>
   rotation=<rot -4.2969 -90 -3.4316>
  }]
 dyn_Rope_l_03 [{
   shapeType=Capsule
   length=8.505
   radius=1.25
   shapeName=dyn_Rope_l_03_capsule
   center=<0.24 0.4129 -4.4083>
   rotation=<rot -3.1258 -90 -1.5165>
  }]
 dyn_Rope_r_01 [{
   shapeType=Capsule
   length=8.755
   radius=1
   shapeName=dyn_Rope_r_01_capsule
   center=<-0.0559 0.7771 -4.514>
   rotation=<rot -6.8919 -90 4.9969>
  }]
 dyn_Rope_r_02 [{
   shapeType=Capsule
   length=8.505
   radius=1.25
   shapeName=dyn_Rope_r_02_capsule
   center=<0.0143 -0.0808 -1.8145>
   rotation=<rot -4.2969 -90 3.4316>
  }]
 dyn_Rope_r_03 [{
   shapeType=Capsule
   length=8.505
   radius=1.25
   shapeName=dyn_Rope_r_03_capsule
   center=<-0.091 0.4129 -4.4083>
   rotation=<rot -3.1258 -90 1.5165>
  }]
 dyn_SkirtB01 [{
   shapeType=Capsule
   length=5.1287
   radius=5.341
   shapeName=dyn_SkirtB03_capsule
   center=<-0.456 1.8879 -3.9053>
   rotation=<rot 1.4396 -152.5844 -177.4954>
  }]
 dyn_SkirtB02 [{
   shapeType=Capsule
   length=5.1287
   radius=5.341
   shapeName=dyn_SkirtB03_capsule
   center=<-1.6277 4.153 -3.7937>
   rotation=<rot 1.4396 -152.5844 -177.4954>
  }]
 dyn_SkirtB03 [{
   shapeType=Capsule
   length=8.8787
   radius=10.091
   shapeName=dyn_SkirtB03_capsule
   center=<0.4773 8.8086 -7.0018>
   rotation=<rot -1.4395 27.4156 -2.5046>
  }]
 dyn_SkirtB04 [{
   shapeType=Capsule
   length=11.7285
   radius=10.3147
   shapeName=dyn_SkirtB04_capsule
   center=<-5.1049 6.6668 -8.3172>
   rotation=<rot -0.2473 -62.4064 -90.1288>
  }]
 dyn_SkirtB05 [{
   shapeType=Capsule
   length=13.3573
   radius=8.6858
   shapeName=dyn_SkirtB05_capsule
   center=<-5.3364 7.1097 -8.315>
   rotation=<rot -0.2473 -62.4064 -90.1288>
  }]
 dyn_SkirtB06 [{
   shapeType=Capsule
   length=12.0126
   radius=8.2614
   shapeName=dyn_SkirtB06_capsule
   center=<-6.7922 6.6984 -9.5587>
   rotation=<rot -0.2473 -62.4064 -90.1288>
  }]
 dyn_SkirtB07 [{
   shapeType=Capsule
   length=9.5757
   radius=8.9026
   shapeName=dyn_SkirtB06_capsule
   center=<-5.4941 5.7694 -9.6624>
   rotation=<rot -0.2473 -62.4064 -90.1288>
  }]
 dyn_SkirtD01 [{shapeType=Capsule, length=5.75, radius=5.5, shapeName=dyn_SkirtD01_capsule, center=<0 0 -5.1609>}]
 dyn_SkirtD02 [{shapeType=Capsule, length=10.9554, radius=6.8793, shapeName=dyn_SkirtD02_capsule, center=<0 10 -6>, rotation=<rot -5 -180 -180>}]
 dyn_SkirtD03 [{shapeType=Capsule, length=21.7998, radius=8.7433, shapeName=dyn_SkirtD03_capsule, center=<0 10.5 -7.5>, rotation=<rot 0 -90 -90>}]
 dyn_SkirtD04 [{shapeType=Capsule, length=21.0498, radius=8.9933, shapeName=dyn_SkirtD04_capsule, center=<0 9.5 -6>, rotation=<rot 0 -90 -90>}]
 dyn_SkirtD05 [{shapeType=Capsule, length=20.5528, radius=12.2433, shapeName=dyn_SkirtD05_capsule, center=<0 13 -7.5>, rotation=<rot 0 -90 -90>}]
 dyn_SkirtD07 [{shapeType=Capsule, length=25.3565, radius=10.6867, shapeName=dyn_SkirtD07_capsule, center=<0 6 -8>, rotation=<rot 0 -90 -90>}]
 dyn_SkirtE01 [{shapeType=Capsule, length=5.75, radius=5.5, shapeName=dyn_SkirtE01_capsule, center=<0 0 -5.1609>}]
 dyn_SkirtE02 [{
   shapeType=Capsule
   length=3.8716
   radius=8.1284
   shapeName=dyn_SkirtE02_capsule
   center=<0 4.4233 -4.485>
   rotation=<rot 83.066 -118.404 -118.2284>
  }]
 dyn_SkirtE03 [{shapeType=Capsule, length=22.6447, radius=7.3985, shapeName=dyn_SkirtE03_capsule, center=<0 4.5 -6>, rotation=<rot 0 -90 -90>}]
 dyn_SkirtE04 [{shapeType=Capsule, length=21.6026, radius=8.4406, shapeName=dyn_SkirtE03_capsule, center=<0 5 -6>, rotation=<rot 0 -90 -90>}]
 dyn_SkirtE05 [{shapeType=Capsule, length=19.9856, radius=12.0729, shapeName=dyn_SkirtE05_capsule, center=<0 9 -6.5>, rotation=<rot 0 -90 -90>}]
 dyn_SkirtE07 [{shapeType=Capsule, length=27.7361, radius=12.7611, shapeName=dyn_SkirtE07_capsule, center=<0 4 -6>, rotation=<rot 0 -90 -90>}]
 dyn_SkirtF01 [{shapeType=Capsule, length=5.75, radius=5.5, shapeName=dyn_SkirtF01_capsule, center=<0 0 -5.1609>}]
 dyn_SkirtF02 [{
   shapeType=Capsule
   length=3.5343
   radius=8.4657
   shapeName=dyn_SkirtF02_capsule
   center=<0 4.4233 -4.4851>
   rotation=<rot 83.7383 -77.1354 -77.2096>
  }]
 dyn_SkirtF03 [{
   shapeType=Capsule
   length=22.6447
   radius=7.3985
   shapeName=dyn_SkirtF03_capsule
   center=<0.0332 4.9154 -5.9446>
   rotation=<rot -0.2784 -90.3896 -89.9981>
  }]
 dyn_SkirtF04 [{
   shapeType=Capsule
   length=23.5
   radius=6.5431
   shapeName=dyn_SkirtF04_capsule
   center=<0.0353 5.2176 -5.9432>
   rotation=<rot -0.2784 -90.3896 -89.9981>
  }]
 dyn_SkirtF05 [{
   shapeType=Capsule
   length=18.9702
   radius=11.0729
   shapeName=dyn_SkirtF05_capsule
   center=<0.0262 3.871 -5.022>
   rotation=<rot -0.2784 -90.3896 -89.9981>
  }]
 dyn_SkirtG01 [{shapeType=Capsule, length=5.75, radius=5.5, shapeName=dyn_SkirtG01_capsule, center=<0 0 -5.1609>}]
 dyn_SkirtG02 [{
   shapeType=Capsule
   length=3.8716
   radius=8.1284
   shapeName=dyn_SkirtG02_capsule
   center=<0 4.4233 -4.4851>
   rotation=<rot 83.8498 -83.075 -83.1145>
  }]
 dyn_SkirtG03 [{
   shapeType=Capsule
   length=22.1447
   radius=7.8985
   shapeName=dyn_SkirtG03_capsule
   center=<-2.6758 3.8415 -5.9499>
   rotation=<rot -0.2784 -90.3896 -89.9982>
  }]
 dyn_SkirtG04 [{
   shapeType=Capsule
   length=23.5
   radius=6.5431
   shapeName=dyn_SkirtG04_capsule
   center=<0.0353 5.2177 -5.9432>
   rotation=<rot -0.2784 -90.3896 -89.9982>
  }]
 dyn_SkirtG05 [{
   shapeType=Capsule
   length=9.2202
   radius=12.0729
   shapeName=dyn_SkirtG05_capsule
   center=<0.0624 9.2003 -6.5739>
   rotation=<rot -0.2784 -90.3896 -89.9982>
  }]
 dyn_SkirtG07 [{
   shapeType=Capsule
   length=7.2254
   radius=12.0678
   shapeName=dyn_SkirtG07_capsule
   center=<0.0287 4.2494 -5.7966>
   rotation=<rot -0.2784 -90.3896 -89.9982>
  }]
 dyn_SkirtH01 [{shapeType=Capsule, length=5.75, radius=5.5, shapeName=dyn_SkirtH01_capsule, center=<0 0 -5.1609>}]
 dyn_SkirtH02 [{
   shapeType=Capsule
   length=11.2054
   radius=6.6293
   shapeName=dyn_SkirtD02_capsule
   center=<-0.0792 8.555 -5.8337>
   rotation=<rot -4.9999 -180 -175>
  }]
 dyn_SkirtH03 [{
   shapeType=Capsule
   length=22.2998
   radius=8.2433
   shapeName=dyn_SkirtD03_capsule
   center=<0.0672 9.9245 -7.5631>
   rotation=<rot -10.2785 -90.39 -89.9982>
  }]
 dyn_SkirtH04 [{
   shapeType=Capsule
   length=21.2998
   radius=8.7433
   shapeName=dyn_SkirtD04_capsule
   center=<0.0658 9.7133 -5.9213>
   rotation=<rot -0.2785 -90.3896 -89.9979>
  }]
 dyn_SkirtH05 [{
   shapeType=Capsule
   length=18.0498
   radius=11.9933
   shapeName=dyn_SkirtH05_capsule
   center=<0.1002 14.7654 -7.4746>
   rotation=<rot -0.2785 -90.3897 -89.9983>
  }]
 dyn_SkirtH07 [{
   shapeType=Capsule
   length=20.0498
   radius=9.9933
   shapeName=dyn_SkirtH07_capsule
   center=<0.0419 6.1927 -6.6942>
   rotation=<rot -0.2785 -90.3897 -89.9983>
  }]
 dyn_SkirtJ01 [{shapeType=Capsule, length=5.75, radius=5.5, shapeName=dyn_SkirtJ01_capsule, center=<0 0 -5.1609>}]
 dyn_SkirtJ02 [{shapeType=Capsule, length=5.75, radius=5.5, shapeName=dyn_SkirtJ01_capsule, center=<1.7465 0.9653 -5.1609>}]
 dyn_SkirtJ03 [{
   shapeType=Capsule
   length=8.8787
   radius=10.091
   shapeName=dyn_SkirtB03_capsule
   center=<1.078 8.6294 -7.0322>
   rotation=<rot -2.8437 -22.6033 -0.507>
  }]
 dyn_SkirtJ04 [{shapeType=Capsule, length=11, radius=9.25, shapeName=dyn_SkirtJ04_capsule, center=<4.1089 6.2781 -7.926>, rotation=<rot 90 180 -155.0002>}]
 dyn_SkirtJ05 [{shapeType=Capsule, length=11, radius=9.25, shapeName=dyn_SkirtJ05_capsule, center=<4.5322 7.186 -7.926>, rotation=<rot 90 -90 -65.0002>}]
 dyn_SkirtJ06 [{shapeType=Capsule, length=12, radius=8.25, shapeName=dyn_SkirtJ06_capsule, center=<5.0691 8.3373 -7.926>, rotation=<rot 90 180 -155.0002>}]
 dyn_SkirtJ07 [{
   shapeType=Capsule
   length=12
   radius=8.25
   shapeName=dyn_SkirtJ07_capsule
   center=<4.7923 7.7437 -7.926>
   rotation=<rot 90 -175.2364 -150.2366>
  }]
 dyn_pelvis [{shapeType=Capsule, length=18.25, radius=16.5, shapeName=dyn_pelvis_capsule, center=<2.9551 -7.301 0>, rotation=<rot 90 0 0.0002>}]
 foot_l [{shapeType=Capsule, length=11.4145, radius=14.0342, shapeName=foot_r_capsule, center=<-6.6506 8.0336 1.0356>, rotation=<rot 0 0 82.6562>}]
 foot_r [{shapeType=Capsule, length=11.2077, radius=10.8588, shapeName=foot_r_capsule, center=<3.5892 -7.5638 -0.9746>, rotation=<rot 0 0 82.6562>}]
 pelvis [{shapeType=Capsule, length=13.9555, radius=12.8257, shapeName=pelvis_capsule, center=<-9.9405 1.1824 0>}]
 spine_01 [{shapeType=Capsule, length=3.5, radius=12.5182, shapeName=spine_01_capsule, center=<-3.7991 2.4374 0>}]
 thigh_twist_01_l [{
   shapeType=Capsule
   length=30.5449
   radius=12.9551
   shapeName=thigh_r_capsule
   center=<3.9372 0.0814 2.8763>
   rotation=<rot -89.1701 -179.9991 179.9991>
  }]
 thigh_twist_01_r [{
   shapeType=Capsule
   length=31.6096
   radius=12.6404
   shapeName=thigh_r_capsule
   center=<-3.4496 -0.0974 -1.8405>
   rotation=<rot -89.1701 -179.9991 179.9991>
  }]
}
constraints [
 {bone1Name=dyn_SkirtJ02, bone2Name=dyn_SkirtJ01, swing1LimitDegrees=30, swing2LimitDegrees=30, twistLimitDegrees=60}
 {bone1Name=dyn_SkirtJ03, bone2Name=dyn_SkirtJ02, swing1LimitDegrees=10, swing2LimitDegrees=10, twistLimitDegrees=25}
 {bone1Name=dyn_SkirtJ04, bone2Name=dyn_SkirtJ03, swing1LimitDegrees=13, swing2LimitDegrees=13, twistLimitDegrees=30}
 {bone1Name=dyn_SkirtJ05, bone2Name=dyn_SkirtJ04, swing1LimitDegrees=15, swing2LimitDegrees=15, twistLimitDegrees=35}
 {bone1Name=dyn_SkirtJ06, bone2Name=dyn_SkirtJ05, swing1LimitDegrees=18, swing2LimitDegrees=18, twistLimitDegrees=40}
 {bone1Name=dyn_SkirtH02, bone2Name=dyn_SkirtH01, swing1LimitDegrees=5, swing2LimitDegrees=5, twistLimitDegrees=15}
 {bone1Name=dyn_SkirtH03, bone2Name=dyn_SkirtH02, swing1LimitDegrees=5, swing2LimitDegrees=5, twistLimitDegrees=25}
 {bone1Name=dyn_SkirtH04, bone2Name=dyn_SkirtH03, swing1LimitDegrees=13, swing2LimitDegrees=13, twistLimitDegrees=25}
 {bone1Name=dyn_SkirtH05, bone2Name=dyn_SkirtH04, swing1LimitDegrees=15, swing2LimitDegrees=15, twistLimitDegrees=30}
 {bone1Name=dyn_SkirtG02, bone2Name=dyn_SkirtG01, swing1LimitDegrees=8, swing2LimitDegrees=8, twistLimitDegrees=20}
 {bone1Name=dyn_SkirtG03, bone2Name=dyn_SkirtG02, swing1LimitDegrees=10, swing2LimitDegrees=10, twistLimitDegrees=25}
 {bone1Name=dyn_SkirtG04, bone2Name=dyn_SkirtG03, swing1LimitDegrees=13, swing2LimitDegrees=13, twistLimitDegrees=30}
 {bone1Name=dyn_SkirtG05, bone2Name=dyn_SkirtG04, swing1LimitDegrees=15, swing2LimitDegrees=15, twistLimitDegrees=35}
 {bone1Name=dyn_SkirtF02, bone2Name=dyn_SkirtF01, swing1LimitDegrees=8, swing2LimitDegrees=8, twistLimitDegrees=25}
 {bone1Name=dyn_SkirtF03, bone2Name=dyn_SkirtF02, swing1LimitDegrees=10, swing2LimitDegrees=10, twistLimitDegrees=35}
 {bone1Name=dyn_SkirtF04, bone2Name=dyn_SkirtF03, swing1LimitDegrees=13, swing2LimitDegrees=13, twistLimitDegrees=45}
 {bone1Name=dyn_SkirtF05, bone2Name=dyn_SkirtF04, swing1LimitDegrees=15, swing2LimitDegrees=15, twistLimitDegrees=60}
 {bone1Name=dyn_SkirtE02, bone2Name=dyn_SkirtE01, swing1LimitDegrees=8, swing2LimitDegrees=8, twistLimitDegrees=20}
 {bone1Name=dyn_SkirtE03, bone2Name=dyn_SkirtE02, swing1LimitDegrees=10, swing2LimitDegrees=10, twistLimitDegrees=25}
 {bone1Name=dyn_SkirtE04, bone2Name=dyn_SkirtE03, swing1LimitDegrees=13, swing2LimitDegrees=13, twistLimitDegrees=30}
 {bone1Name=dyn_SkirtE05, bone2Name=dyn_SkirtE04, swing1LimitDegrees=15, swing2LimitDegrees=15, twistLimitDegrees=35}
 {bone1Name=dyn_SkirtD02, bone2Name=dyn_SkirtD01, swing1LimitDegrees=8, swing2LimitDegrees=8, twistLimitDegrees=20}
 {bone1Name=dyn_SkirtD03, bone2Name=dyn_SkirtD02, swing1LimitDegrees=10, swing2LimitDegrees=10, twistLimitDegrees=25}
 {bone1Name=dyn_SkirtD04, bone2Name=dyn_SkirtD03, swing1LimitDegrees=13, swing2LimitDegrees=13, twistLimitDegrees=30}
 {bone1Name=dyn_SkirtD05, bone2Name=dyn_SkirtD04, swing1LimitDegrees=15, swing2LimitDegrees=15, twistLimitDegrees=45}
 {
  bone1Name=dyn_SkirtB02
  bone2Name=dyn_SkirtB01
  swing1LimitDegrees=30
  swing1Motion=Free
  swing2LimitDegrees=30
  swing2Motion=Free
  twistLimitDegrees=60
  twistMotion=Free
 }
 {bone1Name=dyn_SkirtB03, bone2Name=dyn_SkirtB02, swing1LimitDegrees=10, swing2LimitDegrees=10, twistLimitDegrees=25}
 {bone1Name=dyn_SkirtB04, bone2Name=dyn_SkirtB03, swing1LimitDegrees=13, swing2LimitDegrees=13, twistLimitDegrees=30}
 {bone1Name=dyn_SkirtB05, bone2Name=dyn_SkirtB04, swing1LimitDegrees=15, swing2LimitDegrees=15, twistLimitDegrees=35}
 {bone1Name=dyn_SkirtB06, bone2Name=dyn_SkirtB05, swing1LimitDegrees=18, swing2LimitDegrees=18, twistLimitDegrees=40}
 {bone1Name=dyn_SkirtH07, bone2Name=dyn_SkirtH05, swing1LimitDegrees=20, swing2LimitDegrees=20, twistLimitDegrees=45}
 {bone1Name=dyn_SkirtG07, bone2Name=dyn_SkirtG05, swing1LimitDegrees=20, swing2LimitDegrees=20, twistLimitDegrees=45}
 {bone1Name=dyn_SkirtE07, bone2Name=dyn_SkirtE05, swing1LimitDegrees=20, swing2LimitDegrees=20, twistLimitDegrees=45}
 {bone1Name=dyn_SkirtD07, bone2Name=dyn_SkirtD05, swing1LimitDegrees=30, swing2LimitDegrees=30, twistLimitDegrees=60}
 {
  bone1Name=dyn_SkirtG04
  bone2Name=dyn_SkirtF04
  swing1LimitDegrees=20
  swing1Motion=Free
  swing2LimitDegrees=20
  swing2Motion=Free
  twistLimitDegrees=20
  twistMotion=Free
 }
 {
  bone1Name=dyn_SkirtG05
  bone2Name=dyn_SkirtF05
  swing1LimitDegrees=20
  swing1Motion=Free
  swing2LimitDegrees=20
  swing2Motion=Free
  twistLimitDegrees=20
  twistMotion=Free
 }
 {bone1Name=dyn_SkirtB07, bone2Name=dyn_SkirtB06, swing1LimitDegrees=20, swing2LimitDegrees=20, twistLimitDegrees=45}
 {bone1Name=dyn_SkirtJ07, bone2Name=dyn_SkirtJ06, swing1LimitDegrees=20, swing2LimitDegrees=20, twistLimitDegrees=45}
 {
  bone1Name=dyn_SkirtE05
  bone2Name=dyn_SkirtF05
  swing1LimitDegrees=20
  swing1Motion=Free
  swing2LimitDegrees=20
  swing2Motion=Free
  twistLimitDegrees=20
  twistMotion=Free
 }
 {
  bone1Name=dyn_SkirtE04
  bone2Name=dyn_SkirtF04
  swing1LimitDegrees=20
  swing1Motion=Free
  swing2LimitDegrees=20
  swing2Motion=Free
  twistLimitDegrees=20
  twistMotion=Free
 }
 {bone1Name=dyn_Rope_l_01, bone2Name=pelvis, swing1LimitDegrees=1, swing2LimitDegrees=1, twistLimitDegrees=10}
 {bone1Name=dyn_Rope_l_02, bone2Name=dyn_Rope_l_01, swing1LimitDegrees=15, swing2LimitDegrees=15, twistLimitDegrees=45}
 {bone1Name=dyn_Rope_l_03, bone2Name=dyn_Rope_l_02, swing1LimitDegrees=15, swing2LimitDegrees=15, twistLimitDegrees=45}
 {bone1Name=dyn_Rope_r_01, bone2Name=pelvis, swing1LimitDegrees=1, swing2LimitDegrees=1, twistLimitDegrees=10}
 {bone1Name=dyn_Rope_r_02, bone2Name=dyn_Rope_r_01, swing1LimitDegrees=25, swing2LimitDegrees=25, twistLimitDegrees=45}
 {bone1Name=dyn_Rope_r_03, bone2Name=dyn_Rope_r_02, swing1LimitDegrees=25, swing2LimitDegrees=25, twistLimitDegrees=45}
 {
  bone1Name=dyn_SkirtB01
  bone2Name=spine_01
  swing1LimitDegrees=5
  swing1Motion=Free
  swing2LimitDegrees=5
  swing2Motion=Free
  twistLimitDegrees=20
  twistMotion=Free
 }
 {
  bone1Name=dyn_SkirtD01
  bone2Name=spine_01
  swing1LimitDegrees=45
  swing1Motion=Free
  swing2LimitDegrees=45
  swing2Motion=Free
  twistLimitDegrees=45
  twistMotion=Free
 }
 {bone1Name=dyn_SkirtE01, bone2Name=spine_01, swing1LimitDegrees=5, swing2LimitDegrees=5, twistLimitDegrees=10}
 {bone1Name=dyn_SkirtF01, bone2Name=spine_01, swing1LimitDegrees=5, swing2LimitDegrees=5, twistLimitDegrees=10}
 {bone1Name=dyn_SkirtG01, bone2Name=spine_01, swing1LimitDegrees=5, swing2LimitDegrees=5, twistLimitDegrees=10}
 {
  bone1Name=dyn_SkirtH01
  bone2Name=spine_01
  swing1LimitDegrees=45
  swing1Motion=Free
  swing2LimitDegrees=45
  swing2Motion=Free
  twistLimitDegrees=45
  twistMotion=Free
 }
 {
  bone1Name=dyn_SkirtJ01
  bone2Name=spine_01
  swing1LimitDegrees=45
  swing1Motion=Free
  swing2LimitDegrees=45
  swing2Motion=Free
  twistLimitDegrees=45
  twistMotion=Free
 }
 {
  bone1Name=dyn_SkirtG03
  bone2Name=dyn_SkirtF03
  swing1LimitDegrees=20
  swing1Motion=Free
  swing2LimitDegrees=20
  swing2Motion=Free
  twistLimitDegrees=20
  twistMotion=Free
 }
 {
  bone1Name=dyn_SkirtE03
  bone2Name=dyn_SkirtF03
  swing1LimitDegrees=20
  swing1Motion=Free
  swing2LimitDegrees=20
  swing2Motion=Free
  twistLimitDegrees=20
  twistMotion=Free
 }
 {bone1Name=dyn_SkirtB03, bone2Name=dyn_SkirtD03, swing1LimitDegrees=60, swing2LimitDegrees=60, twistLimitDegrees=60}
 {bone1Name=dyn_SkirtJ03, bone2Name=dyn_SkirtH03, swing1LimitDegrees=60, swing2LimitDegrees=60, twistLimitDegrees=60}
 {bone1Name=dyn_SkirtB04, bone2Name=dyn_SkirtD04, swing1LimitDegrees=60, swing2LimitDegrees=60, twistLimitDegrees=60}
 {bone1Name=dyn_SkirtJ04, bone2Name=dyn_SkirtH04, swing1LimitDegrees=60, swing2LimitDegrees=60, twistLimitDegrees=60}
 {bone1Name=dyn_SkirtB05, bone2Name=dyn_SkirtD05, swing1LimitDegrees=60, swing2LimitDegrees=60, twistLimitDegrees=60}
 {bone1Name=dyn_SkirtB07, bone2Name=dyn_SkirtD07, swing1LimitDegrees=60, swing2LimitDegrees=60, twistLimitDegrees=60}
 {bone1Name=dyn_SkirtJ05, bone2Name=dyn_SkirtH05, swing1LimitDegrees=60, swing2LimitDegrees=60, twistLimitDegrees=60}
 {bone1Name=dyn_SkirtJ07, bone2Name=dyn_SkirtH07, swing1LimitDegrees=60, swing2LimitDegrees=60, twistLimitDegrees=60}
 {bone1Name=dyn_SkirtG02, bone2Name=dyn_SkirtH02, swing1LimitDegrees=15, swing2LimitDegrees=15, twistLimitDegrees=60}
 {bone1Name=dyn_SkirtG03, bone2Name=dyn_SkirtH03, swing1LimitDegrees=30, swing2LimitDegrees=30, twistLimitDegrees=60}
 {
  bone1Name=dyn_SkirtG04
  bone2Name=dyn_SkirtH04
  swing1LimitDegrees=45
  swing1Motion=Free
  swing2LimitDegrees=45
  swing2Motion=Free
  twistLimitDegrees=90
  twistMotion=Free
 }
 {
  bone1Name=dyn_SkirtG05
  bone2Name=dyn_SkirtH05
  swing1LimitDegrees=45
  swing1Motion=Free
  swing2LimitDegrees=45
  swing2Motion=Free
  twistLimitDegrees=90
  twistMotion=Free
 }
 {
  bone1Name=dyn_SkirtG07
  bone2Name=dyn_SkirtH07
  swing1LimitDegrees=45
  swing1Motion=Free
  swing2LimitDegrees=45
  swing2Motion=Free
  twistLimitDegrees=90
  twistMotion=Free
 }
 {
  bone1Name=dyn_SkirtE02
  bone2Name=dyn_SkirtD02
  swing1LimitDegrees=45
  swing1Motion=Free
  swing2LimitDegrees=45
  swing2Motion=Free
  twistLimitDegrees=45
  twistMotion=Free
 }
 {
  bone1Name=dyn_SkirtE03
  bone2Name=dyn_SkirtD03
  swing1LimitDegrees=45
  swing1Motion=Free
  swing2LimitDegrees=45
  swing2Motion=Free
  twistLimitDegrees=45
  twistMotion=Free
 }
 {
  bone1Name=dyn_SkirtE04
  bone2Name=dyn_SkirtD04
  swing1LimitDegrees=45
  swing1Motion=Free
  swing2LimitDegrees=45
  swing2Motion=Free
  twistLimitDegrees=45
  twistMotion=Free
 }
 {
  bone1Name=dyn_SkirtE05
  bone2Name=dyn_SkirtD05
  swing1LimitDegrees=45
  swing1Motion=Free
  swing2LimitDegrees=45
  swing2Motion=Free
  twistLimitDegrees=45
  twistMotion=Free
 }
 {
  bone1Name=dyn_SkirtE07
  bone2Name=dyn_SkirtD07
  swing1LimitDegrees=45
  swing1Motion=Free
  swing2LimitDegrees=45
  swing2Motion=Free
  twistLimitDegrees=45
  twistMotion=Free
 }
]
